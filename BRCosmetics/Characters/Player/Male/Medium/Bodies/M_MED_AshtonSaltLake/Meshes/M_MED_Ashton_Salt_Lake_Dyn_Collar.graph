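# Physics asset for the dynamic collar bones of the Ashton Salt Lake medium male body, likely for secondary collar motion.
# Box bodies around the collar are constrained to tiny capsule MSTR anchor bodies, with a three-segment front-left chain and a two-segment front-right chain.
# The front-right chain locks swing1 and twist to act as a hinge, using far larger boxes.
# Head capsules and a right upper-arm capsule are unconstrained, apparently serving as colliders.
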PhysicsAsset /BRCosmetics/Characters/Player/Male/Medium/Bodies/M_MED_AshtonSaltLake/Meshes/M_MED_Ashton_Salt_Lake_Dyn_Collar
bodies {
 dyn_collar_back [{shapeType=Box, extentX=14.2169, extentY=20.6698, extentZ=1.2319, center=<7.1046 0.0001 0.6486>, rotation=<rot 9.7663 0 0>}]
 dyn_collar_back_MSTR [{shapeType=Capsule, length=0.5, radius=0.5, rotation=<rot 90 0 0>}]
 dyn_collar_bk_l [{shapeType=Box, extentX=12.5785, extentY=18.5001, extentZ=1.2319, center=<3.142 0.0001 -0.736>, rotation=<rot -1.2792 0 0>}]
 dyn_collar_bk_l_MSTR [{shapeType=Capsule, length=0.5, radius=0.5, rotation=<rot 90 0 0>}]
 dyn_collar_bk_r [{shapeType=Box, extentX=11.8833, extentY=18.9222, extentZ=1.2319, center=<3.7576 0 0.8159>, rotation=<rot 5.4709 0.0001 -0.0002>}]
 dyn_collar_bk_r_MSTR [{shapeType=Capsule, length=0.5, radius=0.5, rotation=<rot 90 0 0>}]
 dyn_collar_fr_l_01 [{shapeType=Box, extentX=8.7492, extentY=11.1064, extentZ=1.2319, center=<2.92 0.0001 0.3018>, rotation=<rot 10 0 0>}]
 dyn_collar_fr_l_02 [{shapeType=Box, extentX=2.5458, extentY=2.1682, extentZ=0.8797, center=<1.3206 0.0001 0.0199>, rotation=<rot 10 0 0>}]
 dyn_collar_fr_l_03 [{shapeType=Box, extentX=2.5458, extentY=2.1682, extentZ=0.8797, center=<1.3206 0.0001 0.0199>, rotation=<rot 10 0 0>}]
 dyn_collar_fr_l_MSTR [{shapeType=Capsule, length=0.5, radius=0.5, rotation=<rot 90 0 0>}]
 dyn_collar_fr_r_01 [{shapeType=Box, extentX=32, extentY=82, extentZ=2, center=<15 0 4.3833>, rotation=<rot 17.0249 0 0.0001>}]
 dyn_collar_fr_r_02 [{shapeType=Box, extentX=32, extentY=82, extentZ=2, center=<15 0 7.398>, rotation=<rot 28.4849 0 0.0002>}]
 dyn_collar_fr_r_MSTR [{shapeType=Capsule, length=0.5, radius=0.5, rotation=<rot 90 0 0>}]
 dyn_collar_l [{shapeType=Box, extentX=12.6726, extentY=18.3586, extentZ=1.2319, center=<3.002 -0.0322 -0.1635>, rotation=<rot 10 0 3.9058>}]
 dyn_collar_l_MSTR [{shapeType=Capsule, length=0.5, radius=0.5, rotation=<rot 90 0 0>}]
 dyn_collar_r [{shapeType=Box, extentX=12.0337, extentY=17.5813, extentZ=1.7087, center=<3.1257 0.1166 -0.4511>, rotation=<rot 14.5438 -0.7036 -8.6822>}]
 dyn_collar_r_MSTR [{shapeType=Capsule, length=0.5, radius=0.5, rotation=<rot 90 0 0>}]
 head [
  {shapeType=Capsule, length=3.2273, radius=10.6088, shapeName=headMain, center=<3.8757 2.1928 0>, rotation=<rot 90 0 0>}
  {shapeType=Capsule, length=2.9301, radius=3.3816, shapeName=headBack, center=<2.8332 -7.9258 0>}
  {
   shapeType=Capsule
   length=7.3928
   radius=2.0819
   shapeName=headTubeA_l
   center=<0.2151 -5.818 -6.2821>
   rotation=<rot -17.3583 18.1579 -58.6348>
  }
  {
   shapeType=Capsule
   length=6.7678
   radius=2.7069
   shapeName=headTubeA_r
   center=<-4.0161 2.6862 8.4631>
   rotation=<rot 17.3582 18.1615 83.2725>
  }
  {shapeType=Capsule, length=7.4, radius=2.1, shapeName=headTubeB_r, center=<0.2151 -5.818 6.2821>, rotation=<rot 17.3583 18.1579 58.6348>}
 ]
 upperarm_r [{shapeType=Capsule, length=33.0419, radius=4.7447, center=<-16.5011 -0.4668 0.0001>, rotation=<rot 90 0 0>}]
}
constraints [
 {
  bone1Name=dyn_collar_fr_r_01
  bone2Name=dyn_collar_fr_r_MSTR
  swing1LimitDegrees=5
  swing1Motion=Locked
  swing2LimitDegrees=75
  twistLimitDegrees=2
  twistMotion=Locked
 }
 {
  bone1Name=dyn_collar_fr_r_02
  bone2Name=dyn_collar_fr_r_01
  swing1LimitDegrees=1
  swing1Motion=Locked
  swing2LimitDegrees=75
  twistLimitDegrees=0
  twistMotion=Locked
 }
 {bone1Name=dyn_collar_r, bone2Name=dyn_collar_r_MSTR, swing1LimitDegrees=75, swing2LimitDegrees=75, twistLimitDegrees=75}
 {bone1Name=dyn_collar_bk_r, bone2Name=dyn_collar_bk_r_MSTR, swing1LimitDegrees=75, swing2LimitDegrees=75, twistLimitDegrees=75}
 {bone1Name=dyn_collar_back, bone2Name=dyn_collar_back_MSTR, swing1LimitDegrees=90, swing2LimitDegrees=90, twistLimitDegrees=90}
 {bone1Name=dyn_collar_bk_l, bone2Name=dyn_collar_bk_l_MSTR, swing1LimitDegrees=75, swing2LimitDegrees=75, twistLimitDegrees=75}
 {bone1Name=dyn_collar_l, bone2Name=dyn_collar_l_MSTR, swing1LimitDegrees=75, swing2LimitDegrees=75, twistLimitDegrees=75}
 {bone1Name=dyn_collar_fr_l_01, bone2Name=dyn_collar_fr_l_MSTR, swing1LimitDegrees=75, swing2LimitDegrees=75, twistLimitDegrees=75}
 {bone1Name=dyn_collar_fr_l_02, bone2Name=dyn_collar_fr_l_01, swing1LimitDegrees=15, swing2LimitDegrees=45, twistLimitDegrees=5}
 {bone1Name=dyn_collar_fr_l_03, bone2Name=dyn_collar_fr_l_02, swing1LimitDegrees=15, swing2LimitDegrees=45, twistLimitDegrees=5}
]
